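# Unlit parent material for the glowing beam parts of the NuhaiInverted vehicle wheel, tinted by Custom_Color.
# Vertex color red blends a rim glow, built from normal-mapped Fresnel, with a tire glow masked by texture channels.
# The tire animates equalizer-style bars by quantizing UVs into cells and thresholding panned random noise from both V directions.
# Emissive and normal results pass through the shared vehicle damage and de-rez transition functions.
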Material /VehicleCosmetics/Wheels/NuhaiInverted/Materials/M_Parent_NuhaiInverted_Beam
graph {
 properties {
  bAutomaticallySetUsageInEditor=false
  bUseMaterialAttributes=true
  bUsedWithMeshDeformer=true
  bUsedWithSkeletalMesh=true
  previewMesh=/VehicleCosmetics/Wheels/SR_14_T2/Meshes/SK_Wheel_SR_14_T2
  shadingModel=MSM_Unlit
  usedShadingModels=Unlit
 }
 nodes [
  n1 = Add(B=Constant() {r=-0.65})
  Clamp(None=ConstantBiasScale(None=Add(B=@n1)) {bias=-0.125, scale=1024})
  n2 = VectorParameter() {parameterName=Custom_Color, defaultValue=<rgba 0.13 0.56 0.414 1>}
  n3 = TextureSampleParameter2D() {parameterName=Rim_Beam_N, samplerType=SAMPLERTYPE_Normal, texture=/VehicleCosmetics/Wheels/NuhaiInverted/Textures/T_NuhaiInverted_Rim_N}
  n4 = Constant() {r=5}
  n5 = Constant() {r=2}
  n6 = Add(
   A=Multiply(
    A=Constant() {r=2}
    B=MaterialFunctionCall!(Normal=@n3.RGB, Exponent=@n4) {materialFunction=/VehicleCosmetics/Wheels/Cola_Holographic/Materials/MF_Fresnel_Classic}.Result
   )
   B=Multiply(
    A=MaterialFunctionCall(Normal=@n3.RGB, Exponent=@n5) {materialFunction=&MF_Fresnel_Classic}.Result
    B=TextureSampleParameter2D() {parameterName=Rim_Beems2_RGB, samplerType=SAMPLERTYPE_LinearColor, texture=/VehicleCosmetics/Textures/T_Black_Linear}.R
   )
  )
  n7 = Constant() {r=0.5}
  n8 = TextureSample() {samplerType=SAMPLERTYPE_LinearColor, texture=/VehicleCosmetics/Wheels/NuhaiInverted/Textures/T_NuhaiInverted_Tire_RGB}
  n9 = ComponentMask(None=TextureCoordinate()) {g=true, r=true}
  n10 = Floor(None=Multiply(A=Constant() {r=36}, B=ScalarParameter() {defaultValue=1, parameterName=Mask_Tile}))
  n11 = Divide(A=Floor(None=Multiply(A=@n9, B=@n10)), B=@n10)
  n12 = Constant2Vector() {r=1}
  n13 = Time()
  n14 = ScalarParameter() {defaultValue=1, parameterName=EQ_Speed_ThreradEdge}
  n15 = TextureSample(UVs=Panner(Coordinate=Multiply(A=@n12, B=@n11), Time=Multiply(A=@n13, B=@n14)) {speedY=0.25}) {texture=/VehicleCosmetics/Wheels/_SharedTextures/FX/Random_M}
  n16 = Add(A=Add(A=ComponentMask(None=@n11) {g=true}, B=Constant() {desc=-0.812500, r=-0.65}), B=@n15.R)
  n17 = ComponentMask(None=@n11) {g=true}
  n18 = Add(A=@n15.R, B=Add(A=OneMinus(None=@n17), B=Constant() {r=-0.65}))
  n19 = Add(A=Clamp(None=ConstantBiasScale(None=@n16) {bias=-0.125, scale=1024}), B=Clamp(None=ConstantBiasScale(None=@n18) {bias=-0.125, scale=1024}))
  n20 = Constant() {r=20}
  n21 = Multiply(A=@n8.R, B=Clamp(None=Multiply(A=@n19, B=@n20)))
  n22 = TextureSampleParameter2D(UVs=TextureCoordinate()) {
   parameterName=Tire_NuhaiInverted_N
   samplerType=SAMPLERTYPE_LinearColor
   texture=/VehicleCosmetics/Wheels/NuhaiInverted/Textures/T_NuhaiInverted_Tire_N
  }
  n23 = MaterialFunctionCall(Normal=Multiply(A=@n21, B=@n22.RGB), Exponent=Constant() {r=4}) {materialFunction=&MF_Fresnel_Classic}
  n24 = Multiply(A=Clamp(None=Add(A=@n23.Result, B=@n21)), B=Constant() {r=2})
  n25 = Add(A=@n8.G, B=@n21)
  n26 = VertexColor()
  n27 = LinearInterpolate(A=Multiply(A=@n2.RGB, B=Power(Base=@n6, Exp=@n7)), B=Multiply(A=Multiply(A=@n2.RGB, B=@n24), B=Clamp(None=@n25)), Alpha=@n26.R)
  n28 = LinearInterpolate(A=@n3.RGB, B=@n22.RGB, Alpha=@n26.R)
  MaterialFunctionCall(In=MaterialFunctionCall(Material=SetMaterialAttributes!(MaterialAttributes, Emissive Color=@n27, Normal=@n28) {attributeSetTypes [#1, #2]}) {materialFunction=&MF_BaseVehicle_Damage}.Result) {materialFunction=&MF_VehicleDeRez_Transition}
  OneMinus(None=@n19)
 ]
}
